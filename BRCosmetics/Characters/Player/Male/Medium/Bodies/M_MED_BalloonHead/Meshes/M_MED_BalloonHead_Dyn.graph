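# Physics asset for the dangling dynamic parts of the male medium BalloonHead cosmetic body: back, belt, shoulder and wrist strings, torso strap bow and right-arm balloon.
# Small capsules and strap boxes form short chains from MSTR bones, mostly with limited swing and tight twist.
# These chains attach to pelvis or spine_05, mostly with free rotation, as does the balloon to upperarm_twist_01_r.
# Larger thigh capsules provide leg volumes alongside small spine and pelvis anchor bodies.
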PhysicsAsset /BRCosmetics/Characters/Player/Male/Medium/Bodies/M_MED_BalloonHead/Meshes/M_MED_BalloonHead_Dyn
bodies {
 dyn_back_string_01 [{shapeType=Capsule, length=3, radius=0.8, center=<2 0 0>, rotation=<rot 90 0 0>}]
 dyn_back_string_02 [{shapeType=Capsule, length=3, radius=0.8, center=<2 0 0>, rotation=<rot 90 0 0>}]
 dyn_back_string_MSTR [{shapeType=Capsule, length=0.5, radius=0.5}]
 dyn_belt_string_MSTR [{shapeType=Capsule, length=0.5, radius=0.5}]
 dyn_belt_string_long_01 [{shapeType=Capsule, length=3, radius=0.8, center=<1.3 0 0>, rotation=<rot 90 0 0>}]
 dyn_belt_string_long_02 [{shapeType=Capsule, length=3, radius=0.8, center=<1.3 0 0>, rotation=<rot 90 0 0>}]
 dyn_belt_string_short [{shapeType=Capsule, length=3, radius=0.8, center=<1.3 0 0>, rotation=<rot 90 0 0>}]
 dyn_shoulder_string_r [{shapeType=Capsule, length=4.3081, radius=0.6919, center=<0.7039 0 0>, rotation=<rot 90 0 0>}]
 dyn_shoulder_string_r_01 [{shapeType=Capsule, length=4.3081, radius=0.6919, center=<0.7039 0 0>, rotation=<rot 90 0 0>}]
 dyn_shoulder_string_r_01_MSTR [{shapeType=Capsule, length=0.505, radius=0.505, rotation=<rot 0 -90 0>}]
 dyn_shoulder_string_r_MSTR [{shapeType=Capsule, length=0.505, radius=0.505, rotation=<rot 0 -90 0>}]
 dyn_torso_strap_bow_01 [{shapeType=Capsule, length=3, radius=0.6, center=<1 0 0>, rotation=<rot 90 0 0>}]
 dyn_torso_strap_bow_02 [{shapeType=Capsule, length=3, radius=0.6, center=<1 0 0>, rotation=<rot 90 0 0>}]
 dyn_torso_strap_bow_MSTR [{shapeType=Capsule, length=0.5, radius=0.5}]
 dyn_torso_strap_string_bk_01 [{shapeType=Box, extentX=4.1616, extentY=2.5337, extentZ=1.4027, center=<1.4313 0 0>}]
 dyn_torso_strap_string_bk_02 [{shapeType=Box, extentX=4.1616, extentY=2.5337, extentZ=1.4027, center=<1.4313 0 0>}]
 dyn_torso_strap_string_fr_01 [{shapeType=Box, extentX=4.1616, extentY=2.5337, extentZ=1.4027, center=<1.4313 0 0>}]
 dyn_torso_strap_string_fr_02 [{shapeType=Box, extentX=4.1616, extentY=2.5337, extentZ=1.4027, center=<1.4313 0 0>}]
 dyn_upperarm_twist_01_r_balloon [{shapeType=Capsule, length=3.0691, radius=0.5, center=<1.1379 0.3425 6.376>, rotation=<rot 69.9903 -180 -180>}]
 dyn_wrist_string_l_01 [{shapeType=Capsule, length=4.4964, radius=0.5036, center=<2 0 0>, rotation=<rot 90 0 0>}]
 dyn_wrist_string_l_02 [{shapeType=Capsule, length=4.4964, radius=0.5036, center=<2 0 0>, rotation=<rot 90 0 0>}]
 dyn_wrist_string_l_MSTR [{shapeType=Capsule, length=0.5, radius=0.5, rotation=<rot 90 0 0>}]
 pelvis [{shapeType=Capsule, length=0.5, radius=0.5}]
 spine_02 [{shapeType=Capsule, length=0.5, radius=0.5}]
 spine_03 [{shapeType=Capsule, length=0.5, radius=0.5}]
 spine_04 [{shapeType=Capsule, length=0.5, radius=0.5}]
 spine_05 [{shapeType=Capsule, length=0.5, radius=0.5}]
 thigh_l [{shapeType=Capsule, length=9.1899, radius=8.5425, center=<-13.5053 7.0009 1.6157>, rotation=<rot 90 0 0>}]
 thigh_r [{shapeType=Capsule, length=9.1899, radius=8.5425, center=<13.3422 -5.4934 -2.8168>, rotation=<rot 90 -1.1935 -6.1935>}]
 upperarm_twist_01_r [{shapeType=Capsule, length=0.5, radius=0.5}]
}
constraints [
 {bone1Name=dyn_wrist_string_l_MSTR, bone2Name=spine_05, swing1LimitDegrees=45, swing2LimitDegrees=45, twistLimitDegrees=5}
 {bone1Name=dyn_wrist_string_l_01, bone2Name=dyn_wrist_string_l_MSTR, swing1LimitDegrees=45, swing2LimitDegrees=45, twistLimitDegrees=5}
 {bone1Name=dyn_wrist_string_l_02, bone2Name=dyn_wrist_string_l_MSTR, swing1LimitDegrees=45, swing2LimitDegrees=45, twistLimitDegrees=5}
 {bone1Name=dyn_torso_strap_bow_01, bone2Name=dyn_torso_strap_bow_MSTR, swing1LimitDegrees=45, swing2LimitDegrees=45, twistLimitDegrees=5}
 {bone1Name=dyn_torso_strap_bow_02, bone2Name=dyn_torso_strap_bow_MSTR, swing1LimitDegrees=45, swing2LimitDegrees=45, twistLimitDegrees=5}
 {
  bone1Name=dyn_torso_strap_string_bk_01
  bone2Name=dyn_torso_strap_bow_MSTR
  swing1LimitDegrees=30
  swing2LimitDegrees=15
  twistLimitDegrees=5
 }
 {
  bone1Name=dyn_torso_strap_string_bk_02
  bone2Name=dyn_torso_strap_string_bk_01
  swing1LimitDegrees=30
  swing2LimitDegrees=15
  twistLimitDegrees=5
 }
 {
  bone1Name=dyn_torso_strap_string_fr_01
  bone2Name=dyn_torso_strap_bow_MSTR
  swing1LimitDegrees=45
  swing2LimitDegrees=35
  twistLimitDegrees=5
 }
 {
  bone1Name=dyn_torso_strap_string_fr_02
  bone2Name=dyn_torso_strap_string_fr_01
  swing1LimitDegrees=45
  swing2LimitDegrees=35
  twistLimitDegrees=5
 }
 {
  bone1Name=upperarm_twist_01_r
  bone2Name=spine_05
  swing1LimitDegrees=45
  swing1Motion=Free
  swing2LimitDegrees=45
  swing2Motion=Free
  twistLimitDegrees=45
  twistMotion=Free
 }
 {bone1Name=dyn_shoulder_string_r, bone2Name=dyn_shoulder_string_r_MSTR, swing1LimitDegrees=45, swing2LimitDegrees=45, twistLimitDegrees=5}
 {bone1Name=dyn_belt_string_short, bone2Name=dyn_belt_string_MSTR, swing1LimitDegrees=45, swing2LimitDegrees=45, twistLimitDegrees=5}
 {bone1Name=dyn_belt_string_long_01, bone2Name=dyn_belt_string_MSTR, swing1LimitDegrees=45, swing2LimitDegrees=45, twistLimitDegrees=5}
 {bone1Name=dyn_belt_string_long_02, bone2Name=dyn_belt_string_long_01, swing1LimitDegrees=45, swing2LimitDegrees=45, twistLimitDegrees=5}
 {bone1Name=dyn_back_string_01, bone2Name=dyn_back_string_MSTR, swing1LimitDegrees=45, swing2LimitDegrees=45, twistLimitDegrees=5}
 {bone1Name=dyn_back_string_02, bone2Name=dyn_back_string_01, swing1LimitDegrees=45, swing2LimitDegrees=45, twistLimitDegrees=5}
 {
  bone1Name=dyn_torso_strap_bow_MSTR
  bone2Name=spine_05
  swing1LimitDegrees=45
  swing1Motion=Free
  swing2LimitDegrees=45
  swing2Motion=Free
  twistLimitDegrees=45
  twistMotion=Free
 }
 {
  bone1Name=spine_05
  bone2Name=spine_04
  swing1LimitDegrees=45
  swing1Motion=Free
  swing2LimitDegrees=45
  swing2Motion=Free
  twistLimitDegrees=45
  twistMotion=Free
 }
 {
  bone1Name=spine_04
  bone2Name=spine_03
  swing1LimitDegrees=45
  swing1Motion=Free
  swing2LimitDegrees=45
  swing2Motion=Free
  twistLimitDegrees=45
  twistMotion=Free
 }
 {
  bone1Name=spine_03
  bone2Name=spine_02
  swing1LimitDegrees=45
  swing1Motion=Free
  swing2LimitDegrees=45
  swing2Motion=Free
  twistLimitDegrees=45
  twistMotion=Free
 }
 {
  bone1Name=spine_02
  bone2Name=pelvis
  swing1LimitDegrees=45
  swing1Motion=Free
  swing2LimitDegrees=45
  swing2Motion=Free
  twistLimitDegrees=45
  twistMotion=Free
 }
 {
  bone1Name=dyn_belt_string_MSTR
  bone2Name=pelvis
  swing1LimitDegrees=45
  swing1Motion=Free
  swing2LimitDegrees=45
  swing2Motion=Free
  twistLimitDegrees=45
  twistMotion=Free
 }
 {
  bone1Name=dyn_back_string_MSTR
  bone2Name=pelvis
  swing1LimitDegrees=45
  swing1Motion=Free
  swing2LimitDegrees=45
  swing2Motion=Free
  twistLimitDegrees=45
  twistMotion=Free
 }
 {
  bone1Name=dyn_upperarm_twist_01_r_balloon
  bone2Name=upperarm_twist_01_r
  swing1LimitDegrees=45
  swing1Motion=Free
  swing2LimitDegrees=45
  swing2Motion=Free
  twistLimitDegrees=45
  twistMotion=Free
 }
 {
  bone1Name=dyn_shoulder_string_r_MSTR
  bone2Name=spine_05
  swing1LimitDegrees=45
  swing1Motion=Free
  swing2LimitDegrees=45
  swing2Motion=Free
  twistLimitDegrees=45
  twistMotion=Free
 }
 {
  bone1Name=dyn_shoulder_string_r_01_MSTR
  bone2Name=spine_05
  swing1LimitDegrees=45
  swing1Motion=Free
  swing2LimitDegrees=45
  swing2Motion=Free
  twistLimitDegrees=45
  twistMotion=Free
 }
 {
  bone1Name=dyn_shoulder_string_r_01
  bone2Name=dyn_shoulder_string_r_01_MSTR
  swing1LimitDegrees=45
  swing2LimitDegrees=45
  twistLimitDegrees=5
 }
 {
  bone1Name=thigh_l
  bone2Name=pelvis
  swing1LimitDegrees=45
  swing1Motion=Free
  swing2LimitDegrees=45
  swing2Motion=Free
  twistLimitDegrees=45
  twistMotion=Free
 }
 {
  bone1Name=thigh_r
  bone2Name=pelvis
  swing1LimitDegrees=45
  swing1Motion=Free
  swing2LimitDegrees=45
  swing2Motion=Free
  twistLimitDegrees=45
  twistMotion=Free
 }
]
