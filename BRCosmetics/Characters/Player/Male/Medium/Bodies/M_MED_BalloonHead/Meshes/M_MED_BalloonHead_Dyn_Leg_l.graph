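# Physics asset for the BalloonHead player body's left-leg dangling parts: two dynamite sticks, a fuse and a leg cloth.
# Both sticks hang from a string body constrained to spine_01, the fuse joins the first stick, and the cloth joins thigh_l.
# These joints have tight twist limits and varied swing limits, while spine and thigh rotate freely against the pelvis.
# Simple capsules and one box form the bodies, with a large capsule covering the thigh.
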PhysicsAsset /BRCosmetics/Characters/Player/Male/Medium/Bodies/M_MED_BalloonHead/Meshes/M_MED_BalloonHead_Dyn_Leg_l
bodies {
 dyn_dynamite_fuses_l_01 [{shapeType=Capsule, length=4.3503, radius=0.6497, center=<0.1 0 0>, rotation=<rot 90 0 0>}]
 dyn_dynamite_l_01 [{shapeType=Capsule, length=11.4573, radius=2, center=<5 0 0>, rotation=<rot 90 0 0>}]
 dyn_dynamite_l_02 [{shapeType=Capsule, length=11.4573, radius=2, center=<5 0 0>, rotation=<rot 90 0 0>}]
 dyn_dynamite_string_l_01 [{shapeType=Box, extentX=2, extentY=2, extentZ=2}]
 dyn_leg_cloth_l [{shapeType=Capsule, length=3, radius=2, center=<1 0 0>, rotation=<rot 90 0 0>}]
 pelvis [{shapeType=Capsule, length=0.5, radius=0.5}]
 spine_01 [{shapeType=Capsule, length=1, radius=1}]
 thigh_l [{shapeType=Capsule, length=16.8676, radius=9.781, center=<-15.1314 0.0001 0>, rotation=<rot 90 0 0>}]
}
constraints [
 {bone1Name=dyn_dynamite_l_01, bone2Name=dyn_dynamite_string_l_01, swing1LimitDegrees=120, swing2LimitDegrees=45, twistLimitDegrees=5}
 {bone1Name=dyn_dynamite_fuses_l_01, bone2Name=dyn_dynamite_l_01, swing1LimitDegrees=45, swing2LimitDegrees=45, twistLimitDegrees=5}
 {bone1Name=dyn_dynamite_l_02, bone2Name=dyn_dynamite_string_l_01, swing1LimitDegrees=120, swing2LimitDegrees=45, twistLimitDegrees=5}
 {bone1Name=dyn_dynamite_string_l_01, bone2Name=spine_01, swing1LimitDegrees=5, swing2LimitDegrees=45, twistLimitDegrees=5}
 {bone1Name=dyn_leg_cloth_l, bone2Name=thigh_l, swing1LimitDegrees=10, swing2LimitDegrees=10, twistLimitDegrees=5}
 {
  bone1Name=spine_01
  bone2Name=pelvis
  swing1LimitDegrees=45
  swing1Motion=Free
  swing2LimitDegrees=45
  swing2Motion=Free
  twistLimitDegrees=45
  twistMotion=Free
 }
 {
  bone1Name=thigh_l
  bone2Name=pelvis
  swing1LimitDegrees=45
  swing1Motion=Free
  swing2LimitDegrees=45
  swing2Motion=Free
  twistLimitDegrees=45
  twistMotion=Free
 }
]
