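# Physics asset for the BalloonHead player body cosmetic, covering the dangling dynamite and leg cloth bones on the right leg.
# Capsule bodies hang from a small pelvis, thigh and thigh-twist chain, whose joints have free swing and twist.
# A string bone on the thigh twist branches into two dynamite sticks, the first carrying a smaller third piece, with angle-limited joints.
# Four cloth bones (front, back, in, out) attach directly to the thigh with tighter swing limits.
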PhysicsAsset /BRCosmetics/Characters/Player/Male/Medium/Bodies/M_MED_BalloonHead/Meshes/M_MED_BalloonHead_Dyn_Leg_r
bodies {
 dyn_dynamite_r_01 [{shapeType=Capsule, length=11.4573, radius=2, center=<5 0 0>, rotation=<rot 90 0 0>}]
 dyn_dynamite_r_02 [{shapeType=Capsule, length=11.4573, radius=2, center=<5 0 0>, rotation=<rot 90 0 0>}]
 dyn_dynamite_r_03 [{shapeType=Capsule, length=4.3503, radius=0.6497, center=<0.1 0 0>, rotation=<rot 90 0 0>}]
 dyn_dynamite_string_r_01 [{shapeType=Capsule, length=3, radius=2, center=<1 0 0>, rotation=<rot 90 0 0>}]
 dyn_leg_cloth_bk_r [{shapeType=Capsule, length=3, radius=2, center=<1 0 0>, rotation=<rot 90 0 0>}]
 dyn_leg_cloth_fr_r [{shapeType=Capsule, length=3, radius=2, center=<1 0 0>, rotation=<rot 90 0 0>}]
 dyn_leg_cloth_in_r [{shapeType=Capsule, length=3, radius=2, center=<1 0 0>, rotation=<rot 90 0 0>}]
 dyn_leg_cloth_out_r [{shapeType=Capsule, length=3, radius=2, center=<1 0 0>, rotation=<rot 90 0 0>}]
 pelvis [{shapeType=Capsule, length=0.5, radius=0.5}]
 thigh_r [{shapeType=Capsule, length=1, radius=1}]
 thigh_twist_01_r [{shapeType=Capsule, length=0.5, radius=0.5}]
}
constraints [
 {bone1Name=dyn_dynamite_string_r_01, bone2Name=thigh_twist_01_r, swing1LimitDegrees=45, swing2LimitDegrees=45, twistLimitDegrees=5}
 {bone1Name=dyn_dynamite_r_02, bone2Name=dyn_dynamite_string_r_01, swing1LimitDegrees=120, swing2LimitDegrees=45, twistLimitDegrees=5}
 {bone1Name=dyn_dynamite_r_03, bone2Name=dyn_dynamite_r_01, swing1LimitDegrees=45, swing2LimitDegrees=45, twistLimitDegrees=5}
 {bone1Name=dyn_dynamite_r_01, bone2Name=dyn_dynamite_string_r_01, swing1LimitDegrees=120, swing2LimitDegrees=45, twistLimitDegrees=5}
 {
  bone1Name=thigh_twist_01_r
  bone2Name=thigh_r
  swing1LimitDegrees=45
  swing1Motion=Free
  swing2LimitDegrees=45
  swing2Motion=Free
  twistLimitDegrees=45
  twistMotion=Free
 }
 {bone1Name=dyn_leg_cloth_out_r, bone2Name=thigh_r, swing1LimitDegrees=20, swing2LimitDegrees=20, twistLimitDegrees=5}
 {bone1Name=dyn_leg_cloth_fr_r, bone2Name=thigh_r, swing1LimitDegrees=20, swing2LimitDegrees=20, twistLimitDegrees=5}
 {bone1Name=dyn_leg_cloth_in_r, bone2Name=thigh_r, swing1LimitDegrees=20, swing2LimitDegrees=20, twistLimitDegrees=5}
 {bone1Name=dyn_leg_cloth_bk_r, bone2Name=thigh_r, swing1LimitDegrees=20, swing2LimitDegrees=20, twistLimitDegrees=5}
 {
  bone1Name=thigh_r
  bone2Name=pelvis
  swing1LimitDegrees=45
  swing1Motion=Free
  swing2LimitDegrees=45
  swing2Motion=Free
  twistLimitDegrees=45
  twistMotion=Free
 }
]
