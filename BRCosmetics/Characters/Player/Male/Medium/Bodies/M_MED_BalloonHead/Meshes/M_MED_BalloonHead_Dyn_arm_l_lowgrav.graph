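# Physics asset for dynamic balloons on the BalloonHead medium male body's left arm, named as a low-gravity variant, though it also includes one right upper-arm balloon.
# Five finger balloons hang off a hand balloon on hand_l, while upper-arm balloons attach to twist bones, all with limited swing and tight twist.
# Small arm, spine and pelvis bodies form the supporting skeleton, linked by constraints with free swing and twist.
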PhysicsAsset /BRCosmetics/Characters/Player/Male/Medium/Bodies/M_MED_BalloonHead/Meshes/M_MED_BalloonHead_Dyn_arm_l_lowgrav
bodies {
 dyn_hand_balloon [{shapeType=Capsule, length=0, radius=6.9718, center=<7.3561 0 0>, rotation=<rot 90 0 0>}]
 dyn_index_01_l_balloon [{shapeType=Capsule, length=3, radius=2, center=<3 0 0>, rotation=<rot 90 0 0>}]
 dyn_middle_01_l_balloon [{shapeType=Capsule, length=3, radius=2, center=<3 0 0>, rotation=<rot 90 0 0>}]
 dyn_pinky_01_l_balloon [{shapeType=Capsule, length=3, radius=2, center=<3 0 0>, rotation=<rot 90 0 0>}]
 dyn_ring_01_l_balloon [{shapeType=Capsule, length=3, radius=2, center=<3 0 0>, rotation=<rot 90 0 0>}]
 dyn_thumb_01_l_balloon [{shapeType=Capsule, length=3, radius=2, center=<3 0 0>, rotation=<rot 90 0 0>}]
 dyn_upperarm_twist_01_l_balloon [{shapeType=Capsule, length=0.0001, radius=7.109, rotation=<rot 90 0 0>}]
 dyn_upperarm_twist_01_r_balloon [{shapeType=Capsule, length=0.0001, radius=7.109, rotation=<rot 90 0 0>}]
 dyn_upperarm_twist_02_l_balloon [{shapeType=Capsule, length=0.0001, radius=7.109, rotation=<rot 90 0 0>}]
 hand_l [{shapeType=Capsule, length=0.505, radius=0.505, rotation=<rot 0 -90 0>}]
 pelvis [{shapeType=Capsule, length=3, radius=2}]
 spine_02 [{shapeType=Capsule, length=3, radius=2}]
 spine_03 [{shapeType=Capsule, length=3, radius=2}]
 spine_04 [{shapeType=Capsule, length=3, radius=2}]
 spine_05 [{shapeType=Capsule, length=3, radius=2}]
 upperarm_r [{shapeType=Capsule, length=0.5, radius=0.5}]
 upperarm_twist_01_l [{shapeType=Capsule, length=0.5, radius=0.5}]
 upperarm_twist_01_r [{shapeType=Capsule, length=0.5, radius=0.5}]
 upperarm_twist_02_l [{shapeType=Capsule, length=0.5, radius=0.5}]
}
constraints [
 {bone1Name=dyn_thumb_01_l_balloon, bone2Name=dyn_hand_balloon, swing1LimitDegrees=45, swing2LimitDegrees=45, twistLimitDegrees=5}
 {bone1Name=dyn_index_01_l_balloon, bone2Name=dyn_hand_balloon, swing1LimitDegrees=45, swing2LimitDegrees=45, twistLimitDegrees=5}
 {bone1Name=dyn_middle_01_l_balloon, bone2Name=dyn_hand_balloon, swing1LimitDegrees=45, swing2LimitDegrees=45, twistLimitDegrees=5}
 {bone1Name=dyn_pinky_01_l_balloon, bone2Name=dyn_hand_balloon, swing1LimitDegrees=45, swing2LimitDegrees=45, twistLimitDegrees=5}
 {bone1Name=dyn_ring_01_l_balloon, bone2Name=dyn_hand_balloon, swing1LimitDegrees=45, swing2LimitDegrees=45, twistLimitDegrees=5}
 {bone1Name=dyn_upperarm_twist_01_l_balloon, bone2Name=upperarm_twist_01_l, swing1LimitDegrees=45, swing2LimitDegrees=45, twistLimitDegrees=5}
 {bone1Name=dyn_upperarm_twist_02_l_balloon, bone2Name=upperarm_twist_02_l, swing1LimitDegrees=45, swing2LimitDegrees=45, twistLimitDegrees=5}
 {bone1Name=dyn_upperarm_twist_01_r_balloon, bone2Name=upperarm_twist_01_r, swing1LimitDegrees=45, swing2LimitDegrees=45, twistLimitDegrees=5}
 {
  bone1Name=upperarm_twist_01_r
  bone2Name=upperarm_r
  swing1LimitDegrees=45
  swing1Motion=Free
  swing2LimitDegrees=45
  swing2Motion=Free
  twistLimitDegrees=45
  twistMotion=Free
 }
 {bone1Name=dyn_hand_balloon, bone2Name=hand_l, swing1LimitDegrees=45, swing2LimitDegrees=45, twistLimitDegrees=5}
 {
  bone1Name=upperarm_twist_01_l
  bone2Name=spine_05
  swing1LimitDegrees=45
  swing1Motion=Free
  swing2LimitDegrees=45
  swing2Motion=Free
  twistLimitDegrees=45
  twistMotion=Free
 }
 {
  bone1Name=upperarm_twist_02_l
  bone2Name=spine_05
  swing1LimitDegrees=45
  swing1Motion=Free
  swing2LimitDegrees=45
  swing2Motion=Free
  twistLimitDegrees=45
  twistMotion=Free
 }
 {
  bone1Name=upperarm_r
  bone2Name=spine_05
  swing1LimitDegrees=45
  swing1Motion=Free
  swing2LimitDegrees=45
  swing2Motion=Free
  twistLimitDegrees=45
  twistMotion=Free
 }
 {
  bone1Name=spine_05
  bone2Name=spine_04
  swing1LimitDegrees=45
  swing1Motion=Free
  swing2LimitDegrees=45
  swing2Motion=Free
  twistLimitDegrees=45
  twistMotion=Free
 }
 {
  bone1Name=spine_04
  bone2Name=spine_03
  swing1LimitDegrees=45
  swing1Motion=Free
  swing2LimitDegrees=45
  swing2Motion=Free
  twistLimitDegrees=45
  twistMotion=Free
 }
 {
  bone1Name=spine_03
  bone2Name=spine_02
  swing1LimitDegrees=45
  swing1Motion=Free
  swing2LimitDegrees=45
  swing2Motion=Free
  twistLimitDegrees=45
  twistMotion=Free
 }
 {
  bone1Name=spine_02
  bone2Name=pelvis
  swing1LimitDegrees=45
  swing1Motion=Free
  swing2LimitDegrees=45
  swing2Motion=Free
  twistLimitDegrees=45
  twistMotion=Free
 }
 {
  bone1Name=hand_l
  bone2Name=spine_05
  swing1LimitDegrees=45
  swing1Motion=Free
  swing2LimitDegrees=45
  swing2Motion=Free
  twistLimitDegrees=45
  twistMotion=Free
 }
]
