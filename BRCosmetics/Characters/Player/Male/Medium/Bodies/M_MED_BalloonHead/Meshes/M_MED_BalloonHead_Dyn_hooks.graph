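# Physics asset for the dangling hook and spear accessories on the BalloonHead medium male player cosmetic body.
# It has three independent two-body pairs (fish hook, fish spear, long hook), each joining a long, offset capsule to a tiny capsule on its MSTR bone.
# Identical constraints limit each pair to moderate, asymmetric swing with very little twist, with no constraints linking the pairs to each other.
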PhysicsAsset /BRCosmetics/Characters/Player/Male/Medium/Bodies/M_MED_BalloonHead/Meshes/M_MED_BalloonHead_Dyn_hooks
bodies {
 dyn_fish_hook [{shapeType=Capsule, length=8.5084, radius=0.7802, center=<6 0 0>, rotation=<rot 90 0 0>}]
 dyn_fish_hook_MSTR [{shapeType=Capsule, length=0.5, radius=0.5}]
 dyn_fish_spear [{shapeType=Capsule, length=8.5084, radius=0.7802, center=<6 0 0>, rotation=<rot 90 0 0>}]
 dyn_fish_spear_MSTR [{shapeType=Capsule, length=0.5, radius=0.5}]
 dyn_long_hook [{shapeType=Capsule, length=8.5084, radius=0.7802, center=<6 0 0>, rotation=<rot 90 0 0>}]
 dyn_long_hook_MSTR [{shapeType=Capsule, length=0.5, radius=0.5}]
}
constraints [
 {bone1Name=dyn_fish_hook, bone2Name=dyn_fish_hook_MSTR, swing1LimitDegrees=30, swing2LimitDegrees=45, twistLimitDegrees=5}
 {bone1Name=dyn_fish_spear, bone2Name=dyn_fish_spear_MSTR, swing1LimitDegrees=30, swing2LimitDegrees=45, twistLimitDegrees=5}
 {bone1Name=dyn_long_hook, bone2Name=dyn_long_hook_MSTR, swing1LimitDegrees=30, swing2LimitDegrees=45, twistLimitDegrees=5}
]
