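# Physics asset for the BalloonHead medium male character's torso balloons, named as a low-gravity variant for dynamic bone simulation.
# Large capsules on the spine_03 and spine_04 dyn balloon bones are each constrained to small capsules on their matching spine bone, with moderate swing and tight twist limits.
# The spine_04 to spine_03 constraint leaves swing and twist free.
# No gravity setting appears in the asset itself.
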PhysicsAsset /BRCosmetics/Characters/Player/Male/Medium/Bodies/M_MED_BalloonHead/Meshes/M_MED_BalloonHead_Dyn_torso_lowgrav
bodies {
 dyn_spine_03_balloon [{shapeType=Capsule, length=3.71, radius=11.3665, center=<0 0 1.8352>, rotation=<rot -0.0001 0 0>}]
 dyn_spine_04_balloon [{shapeType=Capsule, length=3.71, radius=11.3665, center=<-2.7526 0.0002 0.8348>, rotation=<rot 90 0 0>}]
 spine_03 [{shapeType=Capsule, length=3, radius=2}]
 spine_04 [{shapeType=Capsule, length=3, radius=2}]
}
constraints [
 {bone1Name=dyn_spine_04_balloon, bone2Name=spine_04, swing1LimitDegrees=45, swing2LimitDegrees=45, twistLimitDegrees=5}
 {
  bone1Name=spine_04
  bone2Name=spine_03
  swing1LimitDegrees=45
  swing1Motion=Free
  swing2LimitDegrees=45
  swing2Motion=Free
  twistLimitDegrees=45
  twistMotion=Free
 }
 {bone1Name=dyn_spine_03_balloon, bone2Name=spine_03, swing1LimitDegrees=45, swing2LimitDegrees=45, twistLimitDegrees=5}
]
