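# Animation Blueprint for the BalloonHead male body cosmetic, adding secondary motion to a pose copied from the attached parent mesh.
# Weighted constraints anchor thigh helper, pad, string and prop bones, feeding RigidBody simulations split by body region.
# Low-gravity torso and left-arm passes run before arm-prop constraints and normal-gravity passes.
# CrouchAlpha rotates hook and spear bones, InvCrouchAlpha weights hook physics, and HandOffGunAlpha offsets the left upper arm.
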AnimBlueprint /BRCosmetics/Characters/Player/Male/Medium/Bodies/M_MED_BalloonHead/Meshes/M_MED_BalloonHead_AnimBP
target_skeleton=&/M_MED_BalloonHead_Skeleton
graphs [
 AnimGraph {nodes [
   n1 = "Constraint - thigh_half_l (R)"(ComponentPose=Local To Component(LocalPose=Copy Pose From Mesh() {bUseAttachedParent=true}), Alpha=0.75, ConstraintWeights_0=1) {
    alphaBoolBlend {blendOption=Linear}
    boneToModify {boneName=thigh_half_l}
    constraintSetup [{transformType=Rotation, targetBone {boneName=thigh_l}}]
    constraintWeights [1]
   }
   n2 = "Constraint - thigh_half_r (R)"(ComponentPose=@n1, Alpha=0.75, ConstraintWeights_0=1) {
    alphaBoolBlend {blendOption=Linear}
    boneToModify {boneName=thigh_half_r}
    constraintSetup [{transformType=Rotation, targetBone {boneName=thigh_r}}]
    constraintWeights [1]
   }
   n3 = "Constraint - dyn_kneepad_l (PPP)"!(ComponentPose=@n2, Alpha=1, ConstraintWeights_0=0.5565, ConstraintWeights_1:float=0.0415, ConstraintWeights_2:float=0.4021) -> (Pose) {
    alphaBoolBlend {blendOption=Linear}
    boneToModify {boneName=dyn_kneepad_l}
    constraintSetup [{targetBone {boneName=calf_l}}, {targetBone {boneName=calf_twist_02_l}}, {targetBone {boneName=thigh_l}}]
    constraintWeights [1,1,1]
   }
   n4 = "Constraint - dyn_elbow_pad_r (PPPPP)"!(
    ComponentPose=@n3
    Alpha=1
    ConstraintWeights_0=0.2244
    ConstraintWeights_1:float=0.2693
    ConstraintWeights_2:float=0.4201
    ConstraintWeights_3:float=0.0575
    ConstraintWeights_4:float=0.0288
   ) -> (Pose) {
    alphaBoolBlend {blendOption=Linear}
    boneToModify {boneName=dyn_elbow_pad_r}
    constraintSetup [
     {targetBone {boneName=lowerarm_twist_02_r}}
     {targetBone {boneName=upperarm_r}}
     {targetBone {boneName=lowerarm_r}}
     {targetBone {boneName=lowerarm_twist_01_r}}
     {targetBone {boneName=upperarm_twist_02_r}}
    ]
    constraintWeights [1,1,1,1,1]
   }
   n5 = "Constraint - dyn_belt_string_MSTR (PPPP)"!(
    ComponentPose=@n4
    Alpha=1
    ConstraintWeights_0=0.0523
    ConstraintWeights_1:float=0.1257
    ConstraintWeights_2:float=0.3214
    ConstraintWeights_3:float=0.5006
   ) -> (Pose) {
    alphaBoolBlend {blendOption=Linear}
    boneToModify {boneName=dyn_belt_string_MSTR}
    constraintSetup [
     {targetBone {boneName=spine_01}}
     {targetBone {boneName=thigh_twist_01_l}}
     {targetBone {boneName=thigh_twist_01_r}}
     {targetBone {boneName=pelvis}}
    ]
    constraintWeights [1,1,1,1]
   }
   n6 = "Constraint - dyn_back_string_MSTR (PPPPP)"!(
    ComponentPose=@n5
    Alpha=1
    ConstraintWeights_0=0.1215
    ConstraintWeights_1:float=0.0211
    ConstraintWeights_2:float=0.0251
    ConstraintWeights_3:float=0.0367
    ConstraintWeights_4:float=0.7958
   ) -> (Pose) {
    alphaBoolBlend {blendOption=Linear}
    boneToModify {boneName=dyn_back_string_MSTR}
    constraintSetup [
     {targetBone {boneName=spine_01}}
     {targetBone {boneName=spine_02}}
     {targetBone {boneName=spine_03}}
     {targetBone {boneName=thigh_twist_01_r}}
     {targetBone {boneName=pelvis}}
    ]
    constraintWeights [1,1,1,1,1]
   }
   n7 = "Constraint - dyn_fish_hook_MSTR (PPPPPP)"!(
    ComponentPose=@n6
    Alpha=1
    ConstraintWeights_0=0.0683
    ConstraintWeights_1:float=0.0348
    ConstraintWeights_2:float=0.0194
    ConstraintWeights_3:float=0.5433
    ConstraintWeights_4:float=0.0034
    ConstraintWeights_5:float=0.3309
   ) -> (Pose) {
    alphaBoolBlend {blendOption=Linear}
    boneToModify {boneName=dyn_fish_hook_MSTR}
    constraintSetup [
     {targetBone {boneName=dyn_belt_string_MSTR}}
     {targetBone {boneName=thigh_twist_01_r}}
     {targetBone {boneName=spine_01}}
     {targetBone {boneName=thigh_twist_01_l}}
     {targetBone {boneName=thigh_l}}
     {targetBone {boneName=pelvis}}
    ]
    constraintWeights [1,1,1,1,1,1]
   }
   n8 = "Constraint - dyn_fish_spear_MSTR (PPPPPP)"!(
    ComponentPose=@n7
    Alpha=1
    ConstraintWeights_0=0.1902
    ConstraintWeights_1:float=0.055
    ConstraintWeights_2:float=0.0219
    ConstraintWeights_3:float=0.4178
    ConstraintWeights_4:float=0.0018
    ConstraintWeights_5:float=0.3134
   ) -> (Pose) {
    alphaBoolBlend {blendOption=Linear}
    boneToModify {boneName=dyn_fish_spear_MSTR}
    constraintSetup [
     {targetBone {boneName=dyn_belt_string_MSTR}}
     {targetBone {boneName=thigh_twist_01_r}}
     {targetBone {boneName=spine_01}}
     {targetBone {boneName=thigh_twist_01_l}}
     {targetBone {boneName=thigh_l}}
     {targetBone {boneName=pelvis}}
    ]
    constraintWeights [1,1,1,1,1,1]
   }
   n9 = "Constraint - dyn_long_hook_MSTR (PPPPPP)"!(
    ComponentPose=@n8
    Alpha=1
    ConstraintWeights_0=0.0228
    ConstraintWeights_1:float=0.0121
    ConstraintWeights_2:float=0.0161
    ConstraintWeights_3:float=0.6233
    ConstraintWeights_4:float=0.0059
    ConstraintWeights_5:float=0.3197
   ) -> (Pose) {
    alphaBoolBlend {blendOption=Linear}
    boneToModify {boneName=dyn_long_hook_MSTR}
    constraintSetup [
     {targetBone {boneName=dyn_belt_string_MSTR}}
     {targetBone {boneName=thigh_twist_01_r}}
     {targetBone {boneName=spine_01}}
     {targetBone {boneName=thigh_twist_01_l}}
     {targetBone {boneName=thigh_l}}
     {targetBone {boneName=pelvis}}
    ]
    constraintWeights [1,1,1,1,1,1]
   }
   n10 = "Constraint - dyn_arrow_03 (PP)"!(ComponentPose=@n9, Alpha=1, ConstraintWeights_0=0.8682, ConstraintWeights_1:float=0.1318) -> (Pose) {
    alphaBoolBlend {blendOption=Linear}
    boneToModify {boneName=dyn_arrow_03}
    constraintSetup [{targetBone {boneName=calf_twist_01_r}}, {targetBone {boneName=calf_r}}]
    constraintWeights [1,1]
   }
   n11 = "Constraint - dyn_arrow_02 (PP)"!(ComponentPose=@n10, Alpha=1, ConstraintWeights_0=0.8409, ConstraintWeights_1:float=0.1591) -> (Pose) {
    alphaBoolBlend {blendOption=Linear}
    boneToModify {boneName=dyn_arrow_02}
    constraintSetup [{targetBone {boneName=calf_twist_02_r}}, {targetBone {boneName=calf_r}}]
    constraintWeights [1,1]
   }
   n12 = "Constraint - dyn_arrow_01 (PP)"!(ComponentPose=@n11, Alpha=1, ConstraintWeights_0=0.7185, ConstraintWeights_1:float=0.2815) -> (Pose) {
    alphaBoolBlend {blendOption=Linear}
    boneToModify {boneName=dyn_arrow_01}
    constraintSetup [{targetBone {boneName=calf_twist_02_r}}, {targetBone {boneName=calf_r}}]
    constraintWeights [1,1]
   }
   n13 = "Constraint - dyn_dynamite_string_l_01 (PPP)"!(ComponentPose=@n12, Alpha=1, ConstraintWeights_0=0.0146, ConstraintWeights_1:float=0.5609, ConstraintWeights_2:float=0.4245) -> (Pose) {
    alphaBoolBlend {blendOption=Linear}
    boneToModify {boneName=dyn_dynamite_string_l_01}
    constraintSetup [{targetBone {boneName=spine_01}}, {targetBone {boneName=thigh_twist_01_l}}, {targetBone {boneName=pelvis}}]
    constraintWeights [1,1,1]
   }
   n14 = "Constraint - dyn_dynamite_string_r_01 (PPPPPP)"!(
    ComponentPose=@n13
    Alpha=1
    ConstraintWeights_0=0.0046
    ConstraintWeights_1:float=0.6643
    ConstraintWeights_2:float=0.024
    ConstraintWeights_3:float=0.0026
    ConstraintWeights_4:float=0.0016
    ConstraintWeights_5:float=0.3029
   ) -> (Pose) {
    alphaBoolBlend {blendOption=Linear}
    boneToModify {boneName=dyn_dynamite_string_r_01}
    constraintSetup [
     {targetBone {boneName=thigh_r}}
     {targetBone {boneName=thigh_twist_01_r}}
     {targetBone {boneName=spine_01}}
     {targetBone {boneName=spine_02}}
     {targetBone {boneName=spine_03}}
     {targetBone {boneName=pelvis}}
    ]
    constraintWeights [1,1,1,1,1,1]
   }
   n15 = RigidBody(ComponentPose=@n14) {
    bClampLinearTranslationLimitToRefPose=true
    bOverrideWorldGravity=true
    overridePhysicsAsset=&/M_MED_BalloonHead_Dyn_torso_lowgrav
    simulationSpace=BaseBoneSpace
    alphaBoolBlend {blendOption=Linear}
    baseBoneRef {boneName=spine_03}
    componentAppliedLinearAccClamp=<3000 3000 3000>
    componentLinearAccScale=<0.2 0.2 0.2>
    componentLinearVelScale=<0.2 0.2 0.2>
    overrideWorldGravity {z=-20}
   }
   n16 = RigidBody(ComponentPose=@n15) {
    bClampLinearTranslationLimitToRefPose=true
    bOverrideWorldGravity=true
    overridePhysicsAsset=&/M_MED_BalloonHead_Dyn_arm_l_lowgrav
    simulationSpace=BaseBoneSpace
    alphaBoolBlend {blendOption=Linear}
    baseBoneRef {boneName=spine_05}
    componentAppliedLinearAccClamp=<3000 3000 3000>
    componentLinearAccScale=<0.1 0.1 0.1>
    componentLinearVelScale=<0.1 0.1 0.1>
    overrideWorldGravity {z=-20}
   }
   n17 = Get CrouchAlpha() -> (CrouchAlpha:float)
   n18 = "Transform (Modify) Bone - Bone: dyn_fish_hook_MSTR"(ComponentPose=@n16, Alpha=@n17, Translation="0,0,0", Rotation="40,0,0", Scale="1,1,1") {
    rotationMode=BMM_Additive
    rotationSpace=BCS_BoneSpace
    translationMode=BMM_Additive
    translationSpace=BCS_BoneSpace
    alphaBoolBlend {blendOption=Linear}
    boneToModify {boneName=dyn_fish_hook_MSTR}
    rotation {pitch=40}
   }
   n19 = "Transform (Modify) Bone - Bone: dyn_fish_spear_MSTR"(ComponentPose=@n18, Alpha=@n17, Translation="0,0,0", Rotation="23.548,-28.188,-28.187", Scale="1,1,1") {
    rotationMode=BMM_Additive
    rotationSpace=BCS_BoneSpace
    translationMode=BMM_Additive
    translationSpace=BCS_BoneSpace
    alphaBoolBlend {blendOption=Linear}
    boneToModify {boneName=dyn_fish_spear_MSTR}
    rotation=<rot 23.548 -28.1879 -28.1869>
   }
   n20 = "Transform (Modify) Bone - Bone: dyn_long_hook_MSTR"(ComponentPose=@n19, Alpha=@n17, Translation="0,0,0", Rotation="5.980,28.560,30.698", Scale="1,1,1") {
    rotationMode=BMM_Additive
    rotationSpace=BCS_BoneSpace
    translationMode=BMM_Additive
    translationSpace=BCS_BoneSpace
    alphaBoolBlend {blendOption=Linear}
    boneToModify {boneName=dyn_long_hook_MSTR}
    rotation=<rot 5.98 28.5599 30.6982>
   }
   n21 = "Transform (Modify) Bone - Bone: upperarm_l"(ComponentPose=@n20, Alpha=Get HandOffGunAlpha() -> (HandOffGunAlpha:float), Translation="0,0,0", Rotation="3,-3,-5", Scale="1,1,1") {
    rotationMode=BMM_Additive
    rotationSpace=BCS_BoneSpace
    translationSpace=BCS_BoneSpace
    alphaBoolBlend {blendOption=Linear}
    boneToModify {boneName=upperarm_l}
    rotation=<rot 3 -3 -5>
   }
   n22 = "Constraint - dyn_torso_strap_bow_MSTR (PPPPPPPP)"!(
    ComponentPose=@n21
    Alpha=1
    ConstraintWeights_0=0.0129
    ConstraintWeights_1:float=0.191
    ConstraintWeights_2:float=0.2526
    ConstraintWeights_3:float=0.0381
    ConstraintWeights_4:float=0.0277
    ConstraintWeights_5:float=0.0164
    ConstraintWeights_6:float=0.0684
    ConstraintWeights_7:float=0.3929
   ) -> (Pose) {
    alphaBoolBlend {blendOption=Linear}
    boneToModify {boneName=dyn_torso_strap_bow_MSTR}
    constraintSetup [
     {targetBone {boneName=upperarm_r}}
     {targetBone {boneName=upperarm_twist_01_r}}
     {targetBone {boneName=dyn_upperarm_twist_01_r_balloon}}
     {targetBone {boneName=upperarm_twist_02_r}}
     {targetBone {boneName=spine_03}}
     {targetBone {boneName=spine_04}}
     {targetBone {boneName=spine_05}}
     {targetBone {boneName=clavicle_r}}
    ]
    constraintWeights [1,1,1,1,1,1,1,1]
   }
   n23 = "Constraint - dyn_bear_trap_MSTR (PPPPP)"!(
    ComponentPose=@n22
    Alpha=1
    ConstraintWeights_0=0.0728
    ConstraintWeights_1:float=0.036
    ConstraintWeights_2:float=0.1454
    ConstraintWeights_3:float=0.6835
    ConstraintWeights_4:float=0.0623
   ) -> (Pose) {
    alphaBoolBlend {blendOption=Linear}
    boneToModify {boneName=dyn_bear_trap_MSTR}
    constraintSetup [
     {targetBone {boneName=clavicle_r}}
     {targetBone {boneName=spine_03}}
     {targetBone {boneName=spine_04}}
     {targetBone {boneName=spine_05}}
     {targetBone {boneName=clavicle_l}}
    ]
    constraintWeights [1,1,1,1,1]
   }
   n24 = "Constraint - dyn_wrist_string_l_MSTR (PP)"!(ComponentPose=@n23, Alpha=1, ConstraintWeights_0=0.4322, ConstraintWeights_1:float=0.5678) -> (Pose) {
    alphaBoolBlend {blendOption=Linear}
    boneToModify {boneName=dyn_wrist_string_l_MSTR}
    constraintSetup [{targetBone {boneName=lowerarm_twist_01_l}}, {targetBone {boneName=dyn_hand_balloon}}]
    constraintWeights [1,1]
   }
   n25 = "Constraint - dyn_shoulder_string_r_MSTR (PPPPPPP)"!(
    ComponentPose=@n24
    Alpha=1
    ConstraintWeights_0=0.0377
    ConstraintWeights_1:float=0
    ConstraintWeights_2:float=0.1993
    ConstraintWeights_3:float=0.4214
    ConstraintWeights_4:float=0.1286
    ConstraintWeights_5:float=0.0196
    ConstraintWeights_6:float=0.1933
   ) -> (Pose) {
    alphaBoolBlend {blendOption=Linear}
    boneToModify {boneName=dyn_shoulder_string_r_MSTR}
    constraintSetup [
     {targetBone {boneName=upperarm_r}}
     {targetBone {boneName=lowerarm_r}}
     {targetBone {boneName=upperarm_twist_01_r}}
     {targetBone {boneName=dyn_upperarm_twist_01_r_balloon}}
     {targetBone {boneName=upperarm_twist_02_r}}
     {targetBone {boneName=spine_05}}
     {targetBone {boneName=clavicle_r}}
    ]
    constraintWeights [1,1,1,1,1,1,1]
   }
   n26 = "Constraint - dyn_shoulder_string_r_01_MSTR (PPPPPPP)"!(
    ComponentPose=@n25
    Alpha=1
    ConstraintWeights_0=0.0377
    ConstraintWeights_1:float=0
    ConstraintWeights_2:float=0.1993
    ConstraintWeights_3:float=0.4214
    ConstraintWeights_4:float=0.1286
    ConstraintWeights_5:float=0.0196
    ConstraintWeights_6:float=0.1933
   ) -> (Pose) {
    alphaBoolBlend {blendOption=Linear}
    boneToModify {boneName=dyn_shoulder_string_r_01_MSTR}
    constraintSetup [
     {targetBone {boneName=upperarm_r}}
     {targetBone {boneName=lowerarm_r}}
     {targetBone {boneName=upperarm_twist_01_r}}
     {targetBone {boneName=dyn_upperarm_twist_01_r_balloon}}
     {targetBone {boneName=upperarm_twist_02_r}}
     {targetBone {boneName=spine_05}}
     {targetBone {boneName=clavicle_r}}
    ]
    constraintWeights [1,1,1,1,1,1,1]
   }
   n27 = RigidBody(ComponentPose=@n26, Alpha=Get InvCrouchAlpha() -> (InvCrouchAlpha:float)) {
    bClampLinearTranslationLimitToRefPose=true
    bOverrideWorldGravity=true
    overridePhysicsAsset=&/M_MED_BalloonHead_Dyn_hooks
    simulationSpace=BaseBoneSpace
    alphaBoolBlend {blendOption=Linear}
    baseBoneRef {boneName=pelvis}
    componentAppliedLinearAccClamp=<3000 3000 3000>
    componentLinearAccScale=<0.4 0.2 0.2>
    componentLinearVelScale=<0.4 0.2 0.2>
    overrideWorldGravity {z=-980}
   }
   n28 = RigidBody(ComponentPose=@n27) {
    bClampLinearTranslationLimitToRefPose=true
    bOverrideWorldGravity=true
    overridePhysicsAsset=&/M_MED_BalloonHead_Dyn
    simulationSpace=BaseBoneSpace
    alphaBoolBlend {blendOption=Linear}
    baseBoneRef {boneName=spine_03}
    componentAppliedLinearAccClamp=<3000 3000 3000>
    componentLinearAccScale=<0.4 0.2 0.2>
    componentLinearVelScale=<0.4 0.2 0.2>
    overrideWorldGravity {z=-980}
   }
   n29 = RigidBody(ComponentPose=@n28) {
    bClampLinearTranslationLimitToRefPose=true
    bOverrideWorldGravity=true
    overridePhysicsAsset=&/M_MED_BalloonHead_Dyn_Leg_l
    simulationSpace=BaseBoneSpace
    alphaBoolBlend {blendOption=Linear}
    baseBoneRef {boneName=thigh_half_l}
    componentAppliedLinearAccClamp=<3000 3000 3000>
    componentLinearAccScale=<0.6 0.4 0.4>
    componentLinearVelScale=<0.6 0.4 0.4>
    overrideWorldGravity {z=-980}
   }
   n30 = RigidBody(ComponentPose=@n29) {
    bClampLinearTranslationLimitToRefPose=true
    bOverrideWorldGravity=true
    overridePhysicsAsset=&/M_MED_BalloonHead_Dyn_Leg_r
    simulationSpace=BaseBoneSpace
    alphaBoolBlend {blendOption=Linear}
    baseBoneRef {boneName=thigh_half_r}
    componentAppliedLinearAccClamp=<3000 3000 3000>
    componentLinearAccScale=<0.6 0.4 0.4>
    componentLinearVelScale=<0.6 0.4 0.4>
    overrideWorldGravity {z=-980}
   }
   Output Pose(Result=Component To Local(ComponentPose=@n30)) {layerGroup=None}
  ]}
 EventGraph {links [], nodes []}
]
variables []
