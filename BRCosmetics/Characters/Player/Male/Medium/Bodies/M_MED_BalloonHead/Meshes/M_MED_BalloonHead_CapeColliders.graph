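# Physics asset of cape colliders for the M_MED_BalloonHead medium male player body, giving a cape shapes to collide against.
# Capsules cover the pelvis, spine, neck and clavicles, with extra capsules widening spine_02 and spine_05, and a sphere covers the head.
# Capsules on dynamic arrow and dynamite bones add collision for those accessories.
# Arm, leg and foot bodies have no shapes, and there are no constraints.
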PhysicsAsset /BRCosmetics/Characters/Player/Male/Medium/Bodies/M_MED_BalloonHead/Meshes/M_MED_BalloonHead_CapeColliders
bodies {
 calf_l []
 calf_r []
 clavicle_l [{shapeType=Capsule, length=9.7636, radius=12.6427, center=<15.349 -0.4317 1.7441>, rotation=<rot 75.9352 -0.0001 175.5622>}]
 clavicle_r [{shapeType=Capsule, length=4.2636, radius=14.6427, center=<-14.5643 0.5646 -5.662>, rotation=<rot 75.9351 -0.0001 -4.4378>}]
 dyn_arrow_01 [{
   shapeType=Capsule
   length=11.5
   radius=3.5
   shapeName=dyn_arrow_01_capsule
   center=<7.2543 -0.5728 -0.3301>
   rotation=<rot 80 -89.9998 -89.9999>
  }]
 dyn_arrow_02 [{shapeType=Capsule, length=14.25, radius=4, shapeName=dyn_arrow_02_capsule, center=<9.6547 0 -0.3148>, rotation=<rot -90 -0.0001 0.0002>}]
 dyn_dynamite_l_02 [{
   shapeType=Capsule
   length=14.25
   radius=11.5
   shapeName=dyn_dynamite_l_02_capsule
   center=<5.7305 4.1031 -1.81>
   rotation=<rot -77.6476 -126.9269 146.055>
  }]
 dyn_dynamite_r_03 [{shapeType=Capsule, length=11, radius=12, shapeName=dyn_dynamite_r_03_capsule, center=<-8.3413 0 3.2752>, rotation=<rot -90 24.8577 -14.8575>}]
 foot_l []
 foot_r []
 head [{shapeType=Sphere, radius=17.3125, center=<11.028 -0.174 0>}]
 lowerarm_l []
 lowerarm_r []
 neck_02 [{shapeType=Capsule, length=13.4375, radius=7.3175, center=<-0.2642 -0.3408 0>, rotation=<rot 90 -171.1193 -175.3978>}]
 pelvis [{shapeType=Capsule, length=15.1562, radius=19.7888, center=<-2.6215 -1.3695 0>}]
 spine_01 [{shapeType=Capsule, length=10.7187, radius=12.4138, center=<-2.6215 -0.7156 0>}]
 spine_02 [
  {shapeType=Capsule, length=9.5937, radius=14.445, center=<-2.6215 0.4993 0.0001>}
  {shapeType=Capsule, length=29.875, radius=11.1013, center=<2.5024 0.4992 3.9825>, rotation=<rot 90 0.0049 0.0049>}
  {shapeType=Capsule, length=29.875, radius=11.1013, center=<2.4704 0.4991 -3.9825>, rotation=<rot 90 -0.0396 -0.0396>}
 ]
 spine_03 [{shapeType=Capsule, length=9.1562, radius=10.7575, center=<-1.953 -6.3193 0.0001>}]
 spine_04 [{shapeType=Capsule, length=14.6562, radius=13.6325, center=<2.1467 -1.2748 0.0002>}]
 spine_05 [
  {shapeType=Capsule, length=14.5937, radius=20.445, center=<-1.7514 -0.2997 0>}
  {shapeType=Capsule, length=12.7812, radius=17.0625, center=<0.7875 3.1143 0>}
 ]
 thigh_l []
 thigh_r []
 upperarm_l []
 upperarm_r []
}
constraints []
